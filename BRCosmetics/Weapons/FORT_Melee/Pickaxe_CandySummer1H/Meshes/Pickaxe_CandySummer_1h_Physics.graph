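# Physics asset for the CandySummer one-handed pickaxe, giving its decorative bow bodies and joints for secondary motion.
# A small dyn_bow capsule joins the handle capsule with free swing and twist, and four two-segment chains branch from it.
# The upper chains are boxes with tight swing and wider twist limits.
# The lower chains are capsules that bend on one limited swing axis, with the other swing and twist locked.
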PhysicsAsset /BRCosmetics/Weapons/FORT_Melee/Pickaxe_CandySummer1H/Meshes/Pickaxe_CandySummer_1h_Physics
bodies {
 Melee_TwoHanded_Handle [{shapeType=Capsule, length=1, radius=1}]
 dyn_bow [{shapeType=Capsule, length=1, radius=1}]
 dyn_bow_lo_aa [{shapeType=Capsule, length=0.25, radius=1.25, center=<1 0 0>, rotation=<rot 90 0 0>}]
 dyn_bow_lo_ab [{shapeType=Capsule, length=1.5, radius=1.25, center=<2 0 0>, rotation=<rot 90 0 0>}]
 dyn_bow_lo_ba [{shapeType=Capsule, length=0.25, radius=1.25, center=<1 0 0>, rotation=<rot 90 0 0>}]
 dyn_bow_lo_bb [{shapeType=Capsule, length=1.5, radius=1.25, center=<2 0 0>, rotation=<rot 90 0 0>}]
 dyn_bow_up_aa [{shapeType=Box, extentX=6.5, extentY=4.5, extentZ=2.5, center=<0 0 1.3868>}]
 dyn_bow_up_ab [{shapeType=Box, extentX=6.5, extentY=4.5, extentZ=4, center=<0 0 2.0548>}]
 dyn_bow_up_ba [{shapeType=Box, extentX=6.5, extentY=4.5, extentZ=2.5, center=<0 0 -1.2582>}]
 dyn_bow_up_bb [{shapeType=Box, extentX=6.5, extentY=4.5, extentZ=3.5, center=<0 0 -1.6287>}]
}
constraints [
 {
  bone1Name=dyn_bow
  bone2Name=Melee_TwoHanded_Handle
  swing1LimitDegrees=45
  swing1Motion=Free
  swing2LimitDegrees=45
  swing2Motion=Free
  twistLimitDegrees=45
  twistMotion=Free
 }
 {bone1Name=dyn_bow_up_aa, bone2Name=dyn_bow, swing1LimitDegrees=5, swing2LimitDegrees=5, twistLimitDegrees=15}
 {bone1Name=dyn_bow_up_ab, bone2Name=dyn_bow_up_aa, swing1LimitDegrees=5, swing2LimitDegrees=5, twistLimitDegrees=30}
 {bone1Name=dyn_bow_up_ba, bone2Name=dyn_bow, swing1LimitDegrees=5, swing2LimitDegrees=5, twistLimitDegrees=15}
 {bone1Name=dyn_bow_up_bb, bone2Name=dyn_bow_up_ba, swing1LimitDegrees=5, swing2LimitDegrees=5, twistLimitDegrees=30}
 {
  bone1Name=dyn_bow_lo_aa
  bone2Name=dyn_bow
  swing1LimitDegrees=10
  swing2LimitDegrees=45
  swing2Motion=Locked
  twistLimitDegrees=45
  twistMotion=Locked
 }
 {
  bone1Name=dyn_bow_lo_ab
  bone2Name=dyn_bow_lo_aa
  swing1LimitDegrees=20
  swing2LimitDegrees=45
  swing2Motion=Locked
  twistLimitDegrees=45
  twistMotion=Locked
 }
 {
  bone1Name=dyn_bow_lo_ba
  bone2Name=dyn_bow
  swing1LimitDegrees=10
  swing2LimitDegrees=45
  swing2Motion=Locked
  twistLimitDegrees=45
  twistMotion=Locked
 }
 {
  bone1Name=dyn_bow_lo_bb
  bone2Name=dyn_bow_lo_ba
  swing1LimitDegrees=20
  swing2LimitDegrees=45
  swing2Motion=Locked
  twistLimitDegrees=45
  twistMotion=Locked
 }
]
