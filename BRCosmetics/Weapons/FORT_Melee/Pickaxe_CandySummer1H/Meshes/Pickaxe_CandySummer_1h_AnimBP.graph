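# Animation Blueprint for the one-handed CandySummer pickaxe cosmetic, orienting the weapon by hand and adding secondary physics motion.
# On initialization it compares the mesh component's name to WeaponMeshOffhand and sets offset_weapon true for any other component.
# That flag enables an additive 180-degree bone-space yaw on b_AR_Root, flipping every non-offhand copy.
# A RigidBody node then simulates the override physics asset with its own gravity and dampened component-motion influence.
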AnimBlueprint /BRCosmetics/Weapons/FORT_Melee/Pickaxe_CandySummer1H/Meshes/Pickaxe_CandySummer_1h_AnimBP
target_skeleton=&/Pickaxe_CandySummer_1h_Skeleton
graphs [
 AnimGraph {nodes [
   n1 = "Transform (Modify) Bone - Bone: b_AR_Root"!(
    ComponentPose=Local To Component(LocalPose=Input Pose())
    bAlphaBoolEnabled:bool=Get offset_weapon() -> (offset_weapon:bool)
    Translation="0,0,0"
    Rotation="0,180,0"
    Scale="1,1,1"
   ) -> (Pose) {
    alphaInputType=Bool
    rotationMode=BMM_Additive
    rotationSpace=BCS_BoneSpace
    translationMode=BMM_Additive
    translationSpace=BCS_BoneSpace
    alphaBoolBlend {blendOption=Linear}
    boneToModify {boneName=b_AR_Root}
    rotation {yaw=180}
   }
   n2 = RigidBody(ComponentPose=@n1) {
    bClampLinearTranslationLimitToRefPose=true
    bOverrideWorldGravity=true
    overridePhysicsAsset=&/Pickaxe_CandySummer_1h_Physics
    alphaBoolBlend {blendOption=Linear}
    componentLinearAccScale=<0.3 0.3 0.3>
    componentLinearVelScale=<0.3 0.3 0.3>
    overrideWorldGravity {z=-980}
   }
   Output Pose(Result=Component To Local(ComponentPose=@n2)) {layerGroup=None}
  ]}
 EventGraph {nodes [
   Event BlueprintInitializeAnimation()
   > n1 = Set Mesh_Name(Mesh_Name:Skeletal Mesh Component Object Reference=GetPartSkeletalMeshComponent()) -> (Output_Get:Skeletal Mesh Component Object Reference)
   > n2 = Branch(Condition="Equal Exactly (String)"(A="To String (Object)"!(InObj:Object Reference=@n1) -> (ReturnValue:String), B=WeaponMeshOffhand))
   > Set offset_weapon(offset_weapon:bool=false) -> (Output_Get:bool)
   Set offset_weapon(execute=@n2.else, offset_weapon=true)
   TryGetPawnOwner()
   Event BlueprintUpdateAnimation()
  ]}
]
variables [
 Mesh_Name {type=SkeletalMeshComponent}
 offset_weapon {type=boolean}
]
